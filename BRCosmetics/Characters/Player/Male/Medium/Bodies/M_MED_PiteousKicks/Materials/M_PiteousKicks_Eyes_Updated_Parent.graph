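# Parent eye material for the PiteousKicks male medium character body, built on material attributes for skeletal mesh and mesh deformer use.
# MF_eyes_PiteousKicks_Update supplies the base eye shading, and its Base Color is multiplied by an "Eye Diffuse Multiplier" scalar parameter.
# The modified attributes then pass through MF_TechArtOverrides.
# Shading model is set to come from the material expression, with Default Lit as the one in use.
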Material /BRCosmetics/Characters/Player/Male/Medium/Bodies/M_MED_PiteousKicks/Materials/M_PiteousKicks_Eyes_Updated_Parent
graph {
 properties {
  bUseMaterialAttributes=true
  bUsedWithMeshDeformer=true
  bUsedWithSkeletalMesh=true
  pixelDepthOffsetMode=PDOM_Legacy
  shadingModel=MSM_FromMaterialExpression
  usedShadingModels=Default Lit
 }
 nodes [
  n1 = MaterialFunctionCall() {materialFunction=/BRCosmetics/Characters/Player/Male/Medium/Bodies/M_MED_PiteousKicks/Materials/MaterialFunctions/MF_eyes_PiteousKicks_Update}
  n2 = GetMaterialAttributes(None=@n1.Result) {attributeGetTypes [#1]}
  n3 = ScalarParameter() {defaultValue=1, group=Eye Shader, parameterName=Eye Diffuse Multiplier, sortPriority=1}
  MaterialFunctionCall(In=SetMaterialAttributes!(MaterialAttributes=@n2.MaterialAttributes, Base Color=Multiply(A=@n2.Base Color, B=@n3)) {attributeSetTypes [#1]}) {materialFunction=&MF_TechArtOverrides}
 ]
}
