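# Control Rig for the PiteousKicks male medium character's face accessory part, applying a pose-driven correction to the scarf.
# A spherical pose reader on upperarm_r writes its output to Upperarm_R_Back, with debug drawing switched on.
# The sequence then smooths that value through AlphaInterp and uses it to weight a small rotation of dyn_scarf_01.
# Only that scarf bone is modified; the hair bones in the hierarchy are untouched.
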ControlRigBlueprint /BRCosmetics/Characters/Player/Male/Medium/Bodies/M_MED_PiteousKicks/Meshes/Parts/M_MED_PiteousKicks_FaceAcc_CtrlRig
graphs [
 RigVMFunctionLibrary {links [], nodes []}
 RigVMModel {nodes [
   RigUnit_BeginExecution = &RigUnit_BeginExecution()
   RigVMFunction_Sequence = &RigVMFunction_Sequence(ExecuteContext=@RigUnit_BeginExecution.ExecutePin)
   n1 = &RigUnit_SphericalPoseReader(
    ActiveRegionScaleFactors=(PositiveWidth=1,NegativeWidth=1,PositiveHeight=1,NegativeHeight=1)
    ActiveRegionSize=0.1
    Debug=(bDrawDebug=true,DebugScale=25,DebugSegments=20,DebugThickness=0.25)
    DriverAxis=(X=-0.5,Y=0,Z=-1)
    DriverItem=(Type=Bone,Name="upperarm_r")
    FalloffRegionScaleFactors=(PositiveWidth=0.7,NegativeWidth=0.7,PositiveHeight=0.7,NegativeHeight=0.7)
    FalloffSize=0.2
    FlipHeightScaling=true
    OptionalParentItem=(Type=Bone)
    RotationOffset=(X=-180.3002,Y=-52.0034,Z=-8.3834)
    ExecutePin=@RigVMFunction_Sequence.A
   )
   > RigVMVariableNode(Value=@n1.OutputParam) {variable=Upperarm_R_Back}
   n2 = &RigVMFunction_AlphaInterp(InterpSpeedDecreasing=10, InterpSpeedIncreasing=10, Scale=1, bClampResult=False, Value=RigVMVariableNode() {variable=Upperarm_R_Back}.Value)
   &RigUnit_ModifyTransforms(
    ItemToModify=((Item=(Type=Bone,Name="dyn_scarf_01"),Transform=(Rotation=(X=0,Y=0,Z=0.0436,W=0.999),Translation=(X=0,Y=0,Z=0),Scale3D=(X=1,Y=1,Z=1))))
    WeightMinimum=0
    Weight=@n2.Result
    ExecutePin=@RigVMFunction_Sequence.B
   )
  ]}
]
hierarchy [
 root {loc=<0 0 0>, rot=<0 0 0 1>}
 pelvis {parent=root, loc=<0 -0.7071 93.1191>, rot=<0.0127 0.707 -0.0127 -0.707>}
 spine_01 {parent=pelvis, loc=<6.7494 -0.6388 0>, rot=<0 0 0.034 -0.9994>}
 spine_02 {parent=spine_01, loc=<7.2518 -0.035 0>, rot=<0 0 0.0238 -0.9997>}
 spine_03 {parent=spine_02, loc=<6.0104 0 0>, rot=<0 0 0.0149 -0.9999>}
 spine_04 {parent=spine_03, loc=<8.1024 0 0>, rot=<0 0 -0.0132 -0.9999>}
 spine_05 {parent=spine_04, loc=<10.0146 0 0>, rot=<0 0 -0.0015 -1>}
 neck_01 {parent=spine_05, loc=<11.3138 0.1529 0>, rot=<0 0 -0.2003 -0.9797>}
 neck_02 {parent=neck_01, loc=<4.5627 0 0>, rot=<0 0 0.0504 -0.9987>}
 head {parent=neck_02, loc=<5.3914 0 0>, rot=<0 0 0.115 -0.9934>}
 dyn_hair_fr_top {parent=head, loc=<9.7701 11.2833 1.8814>, rot=<0.1018 0.656 0.7479 -0.0031>}
 dyn_hair_fr_r_01 {parent=head, loc=<6.9242 8.9532 5.608>, rot=<-0.1716 0.2865 0.9409 -0.0559>}
 dyn_hair_side_r_01 {parent=head, loc=<5.2552 7.6654 6.9962>, rot=<0 0 0.9992 0.0395>}
 dyn_hair_side_r_02 {parent=head, loc=<4.2997 4.0857 6.6453>, rot=<0 0 1 -0.0041>}
 dyn_hair_side_l_01 {parent=head, loc=<6.3119 5.0586 -6.1183>, rot=<-0.0802 -0.9626 -0.2579 -0.0215>}
 dyn_hair_side_l_02 {parent=head, loc=<5.1174 2.7996 -6.5331>, rot=<0.0041 -0.9914 -0.1305 0.0005>}
 clavicle_l {parent=spine_05, loc=<4.3575 -0.1466 -2.3223>, rot=<0.0465 -0.7289 -0.009 -0.6829>}
 upperarm_l {parent=clavicle_l, loc=<15.5111 0 0>, rot=<0.0226 -0.3234 -0.0088 -0.9459>}
 upperarm_twist_01_l {parent=upperarm_l, loc=<0.0041 0 0>, rot=<0 0 0 1>}
 upperarm_twist_02_l {parent=upperarm_l, loc=<13.7941 0 0>, rot=<0 0 0 1>}
 clavicle_r {parent=spine_05, loc=<4.3575 -0.1466 2.3223>, rot=<0.7289 0.0465 -0.6829 0.009>}
 upperarm_r {parent=clavicle_r, loc=<-15.5111 0 0>, rot=<0.0226 -0.3234 -0.0088 -0.9459>}
 upperarm_twist_01_r {parent=upperarm_r, loc=<-0.0041 0 0>, rot=<0 0 0 -1>}
 upperarm_twist_02_r {parent=upperarm_r, loc=<-13.7941 0 0>, rot=<0 0 0 -1>}
 dyn_scarf_01 {parent=spine_05, loc=<9.4441 -9.9833 8.9366>, rot=<0.1724 -0.6427 0.7294 -0.1587>}
 dyn_scarf_02 {parent=dyn_scarf_01, loc=<5.7289 0 0>, rot=<0 0 0 1>}
]
variables [Upperarm_R_Back {type=double, default=0}]
