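# Post-copy AnimBlueprint for the PiteousKicks face accessory mesh, animating its scarf and hair on the character.
# It layers the body mesh's pose from both upper arms down over the attached parent's pose and curves, then runs a Control Rig.
# Chained AnimDynamics nodes add spring and gravity motion to scarf and hair bones, with hair simulated relative to root.
# On customization load it caches Body Mesh and sets Backpack Alpha, which the AnimGraph never visibly reads.
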AnimBlueprint /BRCosmetics/Characters/Player/Male/Medium/Bodies/M_MED_PiteousKicks/Meshes/Parts/M_MED_PiteousKicks_FaceAcc_AnimBP
target_skeleton=&/M_MED_PiteousKicks_FaceAcc_Skeleton
graphs [
 AnimGraph {nodes [
   Copy Pose From Mesh() {bUseAttachedParent=true}
   n1 = Get Body Mesh() -> (Body Mesh:Skeletal Mesh Component Object Reference)
   n2 = Control Rig(Source=Layered blend per bone(
     BasePose=Copy Pose From Mesh() {bCopyCurves=true, bUseAttachedParent=true}
     BlendPoses_0=Copy Pose From Mesh(SourceMeshComponent=@n1) {bUseAttachedParent=true}
    ) {layerSetup [{branchFilters [{boneName=upperarm_l}, {boneName=upperarm_r}]}]}) {
    alphaBoolBlend {blendOption=Linear}
    controlRigAssetReference {blueprintRigClass=&/M_MED_PiteousKicks_FaceAcc_CtrlRig.M_MED_PiteousKicks_FaceAcc_CtrlRig_C}
   }
   n3 = Anim Dynamics - Bone: dyn_scarf_01(ComponentPose=Local To Component(LocalPose=@n2), Alpha=1, ExternalForce="0,0,0") {
    angularSpringConstant=0.5
    bAngularSpring=true
    bGravityOverrideInSimSpace=true
    bUseGravityOverride=true
    alphaBoolBlend {blendOption=Linear}
    boundBone {boneName=dyn_scarf_01}
    componentAppliedLinearAccClamp=<10000 10000 10000>
    componentLinearAccScale=<1 1 1>
    componentLinearVelScale=<1 1 1>
    gravityOverride {z=-1200}
    physicsBodyDefinitions [{
      boundBone {boneName=dyn_scarf_01}
      boxExtents=<25 25 25>
      constraintSetup {angularLimitsMax=<0 0 1>, angularLimitsMin=<0 -50 -1>, angularTarget=<0 0 0>}
      localJointOffset=<1 0 0>
     }]
   }
   n4 = Anim Dynamics - Bone: dyn_scarf_02(ComponentPose=@n3, Alpha=1, ExternalForce="0,0,0") {
    angularSpringConstant=0.3
    bAngularSpring=true
    bGravityOverrideInSimSpace=true
    bUseGravityOverride=true
    alphaBoolBlend {blendOption=Linear}
    boundBone {boneName=dyn_scarf_02}
    componentAppliedLinearAccClamp=<10000 10000 10000>
    componentLinearAccScale=<1 1 1>
    componentLinearVelScale=<1 1 1>
    gravityOverride {z=-1200}
    physicsBodyDefinitions [{
      boundBone {boneName=dyn_scarf_02}
      boxExtents=<25 25 25>
      constraintSetup {angularLimitsMax=<0 0 1>, angularLimitsMin=<0 -50 -1>, angularTarget=<0 0 0>}
      localJointOffset=<1 0 0>
     }]
   }
   n5 = Anim Dynamics - Bone: dyn_hair_fr_top(ComponentPose=@n4, Alpha=1, ExternalForce="0,0,0") {
    angularDampingOverride=0.9
    angularSpringConstant=0.5
    bAngularSpring=true
    bGravityOverrideInSimSpace=true
    bOverrideAngularDamping=true
    bUseGravityOverride=true
    simulationSpace=BoneRelative
    alphaBoolBlend {blendOption=Linear}
    boundBone {boneName=dyn_hair_fr_top}
    componentAppliedLinearAccClamp=<10000 10000 10000>
    componentLinearAccScale=<1 1 1>
    componentLinearVelScale=<1 1 1>
    gravityOverride {z=-1200}
    physicsBodyDefinitions [{
      boundBone {boneName=dyn_hair_fr_top}
      boxExtents=<15 15 15>
      constraintSetup {angularLimitsMax=<0 0 3>, angularLimitsMin=<0 -50 -3>, angularTarget=<0 0 0>}
      localJointOffset=<1 0 0>
     }]
    relativeSpaceBone {boneName=root}
   }
   n6 = Anim Dynamics - Bone: dyn_hair_fr_r_01(ComponentPose=@n5, Alpha=1, ExternalForce="0,0,0") {
    angularDampingOverride=0.9
    angularSpringConstant=0.9
    bAngularSpring=true
    bGravityOverrideInSimSpace=true
    bOverrideAngularDamping=true
    bUseGravityOverride=true
    simulationSpace=BoneRelative
    alphaBoolBlend {blendOption=Linear}
    boundBone {boneName=dyn_hair_fr_r_01}
    componentAppliedLinearAccClamp=<10000 10000 10000>
    componentLinearAccScale=<1 1 1>
    componentLinearVelScale=<1 1 1>
    gravityOverride {z=-1200}
    physicsBodyDefinitions [{
      boundBone {boneName=dyn_hair_fr_r_01}
      boxExtents=<15 15 15>
      constraintSetup {angularLimitsMax=<0 0 10>, angularLimitsMin=<0 -30 -10>, angularTarget=<0 0 0>}
      localJointOffset=<1 0 0>
     }]
    relativeSpaceBone {boneName=root}
   }
   n7 = Anim Dynamics - Bone: dyn_hair_side_r_01(ComponentPose=@n6, Alpha=1, ExternalForce="0,0,0") {
    angularDampingOverride=0.9
    angularSpringConstant=0.9
    bAngularSpring=true
    bGravityOverrideInSimSpace=true
    bOverrideAngularDamping=true
    bUseGravityOverride=true
    simulationSpace=BoneRelative
    alphaBoolBlend {blendOption=Linear}
    boundBone {boneName=dyn_hair_side_r_01}
    componentAppliedLinearAccClamp=<10000 10000 10000>
    componentLinearAccScale=<1 1 1>
    componentLinearVelScale=<1 1 1>
    gravityOverride {z=-1200}
    physicsBodyDefinitions [{
      boundBone {boneName=dyn_hair_side_r_01}
      boxExtents=<16 16 16>
      constraintSetup {angularLimitsMax=<0 0 10>, angularLimitsMin=<0 -50 -10>, angularTarget=<0 0 0>}
      localJointOffset=<1 0 0>
     }]
    relativeSpaceBone {boneName=root}
   }
   n8 = Anim Dynamics - Bone: dyn_hair_side_r_02(ComponentPose=@n7, Alpha=1, ExternalForce="0,0,0") {
    angularDampingOverride=0.9
    angularSpringConstant=0.9
    bAngularSpring=true
    bGravityOverrideInSimSpace=true
    bOverrideAngularDamping=true
    bUseGravityOverride=true
    simulationSpace=BoneRelative
    alphaBoolBlend {blendOption=Linear}
    boundBone {boneName=dyn_hair_side_r_02}
    componentAppliedLinearAccClamp=<10000 10000 10000>
    componentLinearAccScale=<1 1 1>
    componentLinearVelScale=<1 1 1>
    gravityOverride {z=-1200}
    physicsBodyDefinitions [{
      boundBone {boneName=dyn_hair_side_r_02}
      boxExtents=<16 16 16>
      constraintSetup {angularLimitsMax=<0 0 10>, angularLimitsMin=<0 -50 -10>, angularTarget=<0 0 0>}
      localJointOffset=<1 0 0>
     }]
    relativeSpaceBone {boneName=root}
   }
   n9 = Anim Dynamics - Bone: dyn_hair_side_l_01(ComponentPose=@n8, Alpha=1, ExternalForce="0,0,0") {
    angularDampingOverride=0.9
    angularSpringConstant=0.9
    bAngularSpring=true
    bGravityOverrideInSimSpace=true
    bOverrideAngularDamping=true
    bUseGravityOverride=true
    simulationSpace=BoneRelative
    alphaBoolBlend {blendOption=Linear}
    boundBone {boneName=dyn_hair_side_l_01}
    componentAppliedLinearAccClamp=<10000 10000 10000>
    componentLinearAccScale=<1 1 1>
    componentLinearVelScale=<1 1 1>
    gravityOverride {z=-1200}
    physicsBodyDefinitions [{
      boundBone {boneName=dyn_hair_side_l_01}
      boxExtents=<15 15 15>
      constraintSetup {angularLimitsMax=<0 0 10>, angularLimitsMin=<0 -50 -10>, angularTarget=<0 0 0>}
      localJointOffset=<1 0 0>
     }]
    relativeSpaceBone {boneName=root}
   }
   n10 = Anim Dynamics - Bone: dyn_hair_side_l_02(ComponentPose=@n9, Alpha=1, ExternalForce="0,0,0") {
    angularDampingOverride=0.9
    angularSpringConstant=0.9
    bAngularSpring=true
    bGravityOverrideInSimSpace=true
    bOverrideAngularDamping=true
    bUseGravityOverride=true
    simulationSpace=BoneRelative
    alphaBoolBlend {blendOption=Linear}
    boundBone {boneName=dyn_hair_side_l_02}
    componentAppliedLinearAccClamp=<10000 10000 10000>
    componentLinearAccScale=<1 1 1>
    componentLinearVelScale=<1 1 1>
    gravityOverride {z=-1200}
    physicsBodyDefinitions [{
      boundBone {boneName=dyn_hair_side_l_02}
      boxExtents=<15 15 15>
      constraintSetup {angularLimitsMax=<0 0 10>, angularLimitsMin=<0 -50 -10>, angularTarget=<0 0 0>}
      localJointOffset=<1 0 0>
     }]
    relativeSpaceBone {boneName=root}
   }
   Output Pose(Result=Component To Local(ComponentPose=@n10)) {layerGroup=None}
  ]}
 EventGraph {nodes [
   n1 = Event OnCharacterCustomizationLoaded()
   > Set Body Mesh(Body Mesh:Skeletal Mesh Component Object Reference=Get Array Item with EFort Custom Part Type(Array=@n1.CustomMeshParts)) -> (Output_Get:Skeletal Mesh Component Object Reference)
   > Set Backpack Alpha(Backpack Alpha:double=SelectFloat(A=0.1, B=1, bPickA=CheckForBackbling())) -> (Output_Get:double)
  ]}
]
variables [
 Backpack Alpha {type=number}
 Body Mesh {type=SkeletalMeshComponent}
]
